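; Projectile actor that drops the Ostrich mech from the sky, damaging what it hits and spawning the mech on impact unless destroy-on-drop is set.
; After a delay, a timeline eases it from an angled offset, then it falls.
; The server traces the ground, damages buildings on a timer, knocks back pawns, and triggers final AOE at ground height or on undamageable buildings.
; Clients run a separate height check to hide it.
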(event Custom|MoveFromOffsetPosition
  ;[1 1 1] Move the actor with a timeline to make the mech come in at an angle
  (Variables|AngleOffset|SetOffsetLoc (Transformation|GetActorLocation))
  (Variables|AngleOffset|SetOffsetRot (Transformation|GetActorRotation))
  (bind (_direction _alpha) (Timeline|MoveFromOffset PlayFromStart)
    (:Update
      (Transformation|SetActorTransform (Math|Transform|MakeTransform (Math|Vector|Lerp(Vector) (Variables|AngleOffset|GetOffsetLoc) (Variables|AngleOffset|GetInitialLoc) _alpha) (Math|Rotator|Lerp(Rotator) (Variables|AngleOffset|GetOffsetRot) (Variables|AngleOffset|GetInitialRot) _alpha true))))
    (:Finished
      ;;[1 1 1] Set velocity once we move over
      (Variables|/Script/Engine.MovementComponent|SetVelocity_X (Math|Vector|MakeVector 0.0 0.0 (ScalableFloat|GetValueatLevel (Variables|Rows|GetRow_FallZ_Velocity))) (Variables|Default|GetFortProjectileMovement)))))

(event Custom|StartSecondStage
  ;[1 1 1] Start Second Stage (only audio hooked up for now)
  (Audio|SpawnSoundAttached 0 (Variables|Default|GetAudio)))

(event Custom|ClientTickDistCheck
  ;[1 1 1] ClientDistanceCheck
  (CallFunction|DistanceCheckClient))

(event Custom|StartEveryThing
  ;[1 1 1] Begin Play - Save Out Values and start overlap Timer
  ;;[1 1 1] Move the actor over so that the mech can come in at an angle
  (Transformation|SetActorLocationAndRotation (+ (Variables|AngleOffset|GetInitialLoc) (Variables|AngleOffset|GetSpawnAngleOffset)) (Math|Rotator|CombineRotators (Variables|AngleOffset|GetInitialRot) (Variables|AngleOffset|GetSpawnAngleRotOffset)))
  (if (Networking|IsServer)
    (Utilities|Array|AddUnique (Variables|Destruction|GetOverlapBuildingActorsToIgnore) self)
    (bind _returnvalue (Utilities|Time|SetTimerbyEvent (delegate Custom|DestroyOnTick) (Variables|Destruction|GetTimeTraceTick) true))
    (Variables|Destruction|SetTimer_Trace _returnvalue)
    (else
      (bind _returnvalue_2 (Utilities|Time|SetTimerbyEvent (delegate Custom|ClientTickDistCheck) (Variables|Destruction|GetTimeTraceTick) true))
      (Variables|GroundCheck|SetTimer_TraceClient _returnvalue_2)))
  (Utilities|FlowControl|Delay 0.1)
  ;;[1 1 1] Turn on all the FX
  (Components|Activation|Activate (Variables|Default|GetFallFX) false)
  (Ability|ExecuteGameplayCueReplicated (Variables|Cues|GetGC_Spawn))
  (Components|Activation|Activate (Variables|Default|GetAthenaSpecialActor) false)
  (CallFunction|MovefromOffsetPosition)
  (|CallParentFunction))

(event EventBeginPlay
  ;[1 1 1] Begin Play - Save Out Values and start overlap Timer
  (CallFunction|SetStartingValues)
  (Utilities|Time|SetTimerbyEvent (delegate Custom|StartEveryThing) 2.0)
  (Variables|AngleOffset|SetInitialLoc (Transformation|GetActorLocation))
  (Variables|AngleOffset|SetInitialRot (Transformation|GetActorRotation))
  (if (Networking|IsServer)
    (CallFunction|Traceforworld)))

(event Custom|Traceforworld
  ;[1 1 1] Get trace for the world to find where we should stop
  (bind (_outhit _returnvalue) (Collision|LineTraceByChannel :Start (Transformation|GetActorLocation) :End (+ (Transformation|GetActorLocation) (Variables|GroundCheck|GetLineTraceOffset)) :TraceChannel "TraceTypeQuery10" :bTraceComplex (Variables|Default|GetTraceComplex) :DrawDebugType "None" :bIgnoreSelf true :TraceColor "(R=1.000000,G=0.000000,B=0.000000,A=1.000000)" :TraceHitColor "(R=0.000000,G=1.000000,B=0.000000,A=1.000000)" :DrawTime 1e+02))
  (if _returnvalue
    (Variables|GroundCheck|SetGroundLocation (.Location (Collision|BreakHitResult _outhit)))
    (Variables|GroundCheck|SetFoundHit true)))

(event Custom|DestroyOnTick
  ;[1 1 1] Building Overlap timer to apply damage to buildings when hit.
  (CallFunction|DistanceCheck)
  (if (Variables|Default|GetOverlapCheckOn)
    (bind _outactors (Collision|SphereOverlapActors (+ (Transformation|GetActorLocation) (Variables|Destruction|GetOverlapOffset)) (Variables|Destruction|GetOverlapBuildingRadius) (Variables|Destruction|GetOverlapBuildingObjectType) "/Script/FortniteGame.BuildingActor" (Variables|Destruction|GetOverlapBuildingActorsToIgnore)))
    (for _array_element _outactors
      (Utilities|Casting|CastToBuildingGameplayActor _array_element
        (:CastFailed
          (if (Variables|/Script/Engine.Actor|GetbCanBeDamaged _array_element)
            (GameplayEffects|ApplyGameplayEffectToTarget :self (Ability|GetAbilitySystemComponent self) :GameplayEffectClass (Variables|GEs|GetGE_EnvironmentalDamage) :Target (Ability|GetAbilitySystemComponent _array_element) :Level 0.0)
            (else
              (CallFunction|DoHitNotDamageableAdjustment))))))))

(event Custom|SpawnOst
  (bind _groundlocation (Variables|GroundCheck|GetGroundLocation))
  ;[1 1 1] Spawn OST
  (if (Math|Conversions|ToBoolean(Integer) (Math|Float|Floor (ScalableFloat|GetValueatLevel (Variables|Rows|GetRow_DestroyOnDrop))))
    ;[1 1 1] Don't Spawn The Mech. Destroy it
    (CallFunction|DestroyonDrop)
    (else
      (Utilities|FlowControl|Delay 0.1)
      (Development|PrintString (Utilities|String|ToString(Vector) _groundlocation))
      (bind _returnvalue (Game|SpawnActorfromClass (Variables|Default|GetOst) (Math|Transform|MakeTransform (+ _groundlocation (Variables|Default|GetMechSpawnOffset)) (Transformation|GetWorldRotation (Variables|Default|GetParent))) "AlwaysSpawn" "SelectDefaultAtRuntime" 0 0))
      (Variables|Default|SetSpawnedOst _returnvalue))))

(event Custom|HideAndKill
  ;[1 1 1] Play GC, Hide Actor and Kill
  (Game|Components|ProjectileMovement|StopSimulating (Variables|Default|GetFortProjectileMovement) (Variables|Kill|GetNullHit))
  (Ability|ExecuteGameplayCueWithParamsReplicated (Variables|Cues|GetGC_Destroy) (Ability|GameplayCue|MakeGameplayCueParameters :NormalizedMagnitude 0.0 :RawMagnitude 0.0 :Location (Variables|GroundCheck|GetGroundLocation) :Normal "0,0,0" :GameplayEffectLevel 1 :AbilityLevel 1 :bReplicateLocationWhenUsingMinimalRepProxy false))
  (Collision|SetActorEnableCollision false)
  (Components|Activation|Deactivate (Variables|Default|GetTrail))
  (Audio|Components|Audio|FadeOut (Variables|Default|GetAudio) 0.25 0.0 "Linear" :!class /Script/Engine.AudioComponent)
  (if (Networking|IsServer)
    (Actor|SetLifeSpan 2.0)
    (else
      (Utilities|FlowControl|Delay 0.2)
      (Rendering|SetActorHiddenInGame true))))

(event Custom|DestroyOnDrop
  ;[1 1 1] Destroy Mech On Drop - Play the FX and feedback but no spawn
  (Fortnite|PlayLocalForceFeedbackatLocationMulti (Variables|Default|GetDetonateForceFeedbackNear) (Variables|Default|GetDetonateForceFeedbackFar) (Transformation|GetActorLocation self) 4e+02 1.4e+03 "OstrichDetonate") ; Gamepad Vibrate
  (Ability|ExecuteGameplayCueReplicated (Variables|Cues|GetGC_DestroyOnDrop))
  (Fortnite|ShowSoundIndicatorLocalforTeam :Location (Transformation|GetActorLocation self) :MaxAudibleDistance 2e+03 :IndicatorType "Gunshot" :Team (Team|GetActorTeam self) :Affiliations (Utilities|Array|MakeArray "Hostile" "Neutral" "Friendly") :Instigator self :Tint "(R=0.000000,G=0.000000,B=0.000000,A=0.000000)" :OverrideIcon "/Game/UI/Foundation/Textures/Icons/Navigation/Mech_Visualizer_icon_32.Mech_Visualizer_icon_32" :AttenuationMaxRangeMultiplier 1.0 :AttenuationIntensityMultiplier 1.0))

(event OnComponentBeginOverlap(PawnCollision)
  (params (OverlappedComponent :type Primitive%20Component%20Object%20Reference) (OtherActor :type Actor%20Object%20Reference) (OtherComp :type Primitive%20Component%20Object%20Reference) (OtherBodyIndex :type Integer) (bFromSweep :type Boolean) (SweepResult :type Hit%20Result%20Structure%20%28by%20ref%29))
  ;[1 1 1] On overlap with pawns apply impulse and/or damage
  (if (Networking|IsServer)
    (if (not (Utilities|Array|ContainsItem (Variables|PawnHits|GetHitPawns) OtherActor))
      (bind _asfort_vehicle_interface (Utilities|Casting|CastToFortVehicleInterface OtherActor)
        (:then
          (if (Utilities|NotEqual(Object) OtherActor (Variables|Default|GetSpawnedOst))
            (GameplayEffects|ApplyGameplayEffectToTarget :self (Ability|GetAbilitySystemComponent self) :GameplayEffectClass (Variables|GEs|GetGE_VehicleDamage) :Target (Vehicle|GetVehicleAbilitySystemComponent _asfort_vehicle_interface) :Level 0.0)
            (Utilities|Array|AddUnique (Variables|PawnHits|GetHitPawns) OtherActor)))
        (:CastFailed
          (bind _asfort_player_pawn (Utilities|Casting|CastToFortPlayerPawn OtherActor))
          (Utilities|Array|AddUnique (Variables|PawnHits|GetHitPawns) _asfort_player_pawn)
          (bind (_returnvalue_x _returnvalue_y) (Math|Vector|BreakVector (Math|Vector|Normalize (- (Transformation|GetActorLocation _asfort_player_pawn) (Transformation|GetActorLocation)))))
          (bind (_returnvalue_x_2 _returnvalue_y_2 _returnvalue_z) (Math|Vector|BreakVector (Math|Vector|vector*float (Math|Vector|MakeVector _returnvalue_x _returnvalue_y) (Variables|PawnHits|GetKnockMag))))
          (Character|LaunchCharacter _asfort_player_pawn (Math|Vector|MakeVector _returnvalue_x_2 _returnvalue_y_2 (+ _returnvalue_z (Variables|PawnHits|GetKnockAddedZ))) true true)
          (if (or (not (Team|OnSameTeam _asfort_player_pawn self)) (Utilities|Equal(Object) _asfort_player_pawn (Game|GetInstigator)))
            (GameplayEffects|ApplyGameplayEffectToTarget :self (Ability|GetAbilitySystemComponent self) :GameplayEffectClass (Variables|GEs|GetGE_PlayerDamage) :Target (Ability|GetAbilitySystemComponent _asfort_player_pawn) :Level 0.0)))))))

(event Custom|FinalHit
  ;[1 1 1] Do Final AOE Hit
  (Utilities|Time|ClearandInvalidateTimerbyHandle (Variables|Destruction|GetTimer_Trace))
  (Utilities|Time|ClearandInvalidateTimerbyHandle (Variables|GroundCheck|GetTimer_TraceClient))
  (bind _outactors (Collision|SphereOverlapActors (Variables|GroundCheck|GetGroundLocation) (ScalableFloat|GetValueatLevel (Variables|Rows|GetRow_FinalPawnOvwelap)) (Variables|PawnHits|GetFinalHitObjectType) "/Script/FortniteGame.FortPawn" (Variables|PawnHits|GetHitPawns)))
  (for _array_element _outactors
    (if (Networking|IsServer)
      (bind _asfort_vehicle_interface (Utilities|Casting|CastToFortVehicleInterface _array_element)
        (:then
          (if (Utilities|NotEqual(Object) _array_element (Variables|Default|GetSpawnedOst))
            (GameplayEffects|ApplyGameplayEffectToTarget :self (Ability|GetAbilitySystemComponent self) :GameplayEffectClass (Variables|GEs|GetGE_VehicleDamage) :Target (Vehicle|GetVehicleAbilitySystemComponent _asfort_vehicle_interface) :Level 0.0)
            (Utilities|Array|AddUnique (Variables|PawnHits|GetHitPawns) _array_element)))
        (:CastFailed
          (bind _asfort_player_pawn (Utilities|Casting|CastToFortPlayerPawn _array_element))
          (Utilities|Array|AddUnique (Variables|PawnHits|GetHitPawns) _asfort_player_pawn)
          (bind (_returnvalue_x _returnvalue_y) (Math|Vector|BreakVector (Math|Vector|Normalize (- (Transformation|GetActorLocation _asfort_player_pawn) (Transformation|GetActorLocation)))))
          (bind (_returnvalue_x_2 _returnvalue_y_2 _returnvalue_z) (Math|Vector|BreakVector (Math|Vector|vector*float (Math|Vector|MakeVector _returnvalue_x _returnvalue_y) (Variables|PawnHits|GetKnockMag))))
          (Character|LaunchCharacter _asfort_player_pawn (Math|Vector|MakeVector _returnvalue_x_2 _returnvalue_y_2 (+ _returnvalue_z (Variables|PawnHits|GetKnockAddedZ))) true true)
          (if (or (not (Team|OnSameTeam _asfort_player_pawn self)) (Utilities|Equal(Object) _asfort_player_pawn (Game|GetInstigator)))
            (GameplayEffects|ApplyGameplayEffectToTarget :self (Ability|GetAbilitySystemComponent self) :GameplayEffectClass (Variables|GEs|GetGE_PlayerDamage) :Target (Ability|GetAbilitySystemComponent _asfort_player_pawn) :Level 0.0))))))
  (bind _outactors_2 (Collision|SphereOverlapActors (Variables|GroundCheck|GetGroundLocation) (ScalableFloat|GetValueatLevel (Variables|Rows|GetRow_FinalPawnOvwelap)) (Variables|Destruction|GetOverlapBuildingObjectType) "/Script/FortniteGame.BuildingActor" (Variables|Destruction|GetOverlapBuildingActorsToIgnore)))
  (for _array_element_2 _outactors_2
    (Utilities|Casting|CastToBuildingGameplayActor _array_element_2
      (:CastFailed
        (GameplayEffects|ApplyGameplayEffectToTarget :self (Ability|GetAbilitySystemComponent self) :GameplayEffectClass (Variables|GEs|GetGE_FinalEnvironmentalDamage) :Target (Ability|GetAbilitySystemComponent _array_element_2) :Level 0.0))))
  (CallFunction|SpawnOst)
  (CallFunction|HideandKill))

(fn DistanceCheck
  (bind (_returnvalue_x _returnvalue_y _returnvalue_z) (Math|Vector|BreakVector (Transformation|GetActorLocation)))
  (bind (_returnvalue_x_2 _returnvalue_y_2 _returnvalue_z_2) (Math|Vector|BreakVector (Transformation|GetActorLocation)))
  (bind (_groundlocation_x _groundlocation_y _groundlocation_z) (Math|Vector|BreakVector (Variables|GetGroundLocation_X)))
  (bind (_groundlocation_x_2 _groundlocation_y_2 _groundlocation_z_2) (Math|Vector|BreakVector (Variables|GetGroundLocation_X)))
  (CallFunction|TurnonOverlaps)
  (if (Math|Boolean|ANDBoolean (>= _groundlocation_z_2 _returnvalue_z_2) (Variables|GroundCheck|GetFoundHit) (not (Variables|Kill|GetPendingKill)))
    (Variables|Kill|SetPendingKill true)
    (if (Networking|IsServer)
      (CallFunction|FinalHit))
    (elif (and (>= (+ _groundlocation_z (Variables|GroundCheck|GetSecondStageHeight)) _returnvalue_z) (not (Variables|GroundCheck|GetInSecondStage)))
      (Variables|GroundCheck|SetInSecondStage true)
      ;[1 1 1] Only audio uses second stage for now
      (CallFunction|StartSecondStage))))

(fn DistanceCheckClient
  (bind (_returnvalue_x _returnvalue_y _returnvalue_z) (Math|Vector|BreakVector (Transformation|GetActorLocation)))
  (bind (_groundlocation_x _groundlocation_y _groundlocation_z) (Math|Vector|BreakVector (Variables|GetGroundLocation_X)))
  (if (Math|Boolean|ANDBoolean (>= _groundlocation_z _returnvalue_z) (Variables|GroundCheck|GetFoundHit) (not (Variables|Kill|GetPendingKill)))
    (Variables|Kill|SetPendingKill true)
    (Utilities|Time|ClearandInvalidateTimerbyHandle (Variables|GroundCheck|GetTimer_TraceClient))
    (CallFunction|HideandKill)))

(fn DoHitNotDamageableAdjustment
  (bind (_outhit _returnvalue) (Collision|LineTraceByChannel (+ (Transformation|GetActorLocation) "0, 0,1000.000000") (- (Transformation|GetActorLocation) "0, 0,1000.000000") "TraceTypeQuery10" true))
  (if _returnvalue
    (Variables|GroundCheck|SetGroundLocation (.Location (Collision|BreakHitResult _outhit)))
    (CallFunction|FinalHit))
  (return))

(fn SetStartingValues
  (Variables|Destruction|SetOverlapBuildingRadius (ScalableFloat|GetValueatLevel (Variables|Rows|GetRow_OverlapRadiusEnvironment)))
  (Variables|PawnHits|SetKnockMag (ScalableFloat|GetValueatLevel (Variables|Rows|GetRow_KnockMag)))
  (Variables|PawnHits|SetKnockAddedZ (ScalableFloat|GetValueatLevel (Variables|Rows|GetRow_KnockAddedZ)))
  (Components|Sphere|SetSphereRadius (Variables|Default|GetPawnCollision) (ScalableFloat|GetValueatLevel (Variables|Rows|GetRow_OverlapRadiusPawn)) true))

(fn TurnOnOverlaps
  (bind (_returnvalue_x _returnvalue_y _returnvalue_z) (Math|Vector|BreakVector (Transformation|GetActorLocation)))
  (if (and (not (Variables|Default|GetOverlapCheckOn)) (<= _returnvalue_z (Variables|Default|GetOverlapCheckHeight)))
    (Variables|Default|SetOverlapCheckOn true)))

(fn ConstructionScript
  (if (not (Networking|IsServer))
    (Variables|/Script/Engine.PrimitiveComponent|SetbGenerateOverlapEvents false (Variables|Default|GetPawnCollision)))
  (|CallParentFunction))
