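; Child cheat manager for the Ostrich vehicle plugin, with debug functions that spawn a test mech vehicle or its drop actor.
; SpawnMech places BP_FigmentTestMechVehicle slightly along the controlled pawn's forward vector.
; SpawnMechDrop spawns BGA_Figment_Ostrich_Drop farther out, scaling the forward vector per axis, so its large Z offset depends on the pawn's pitch.
; Both print an on-screen "Spawning Mech" message.
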(fn SpawnMech
  (Game|SpawnActorfromClass "/Figment_OstrichVehicle/Blueprints/BP_FigmentTestMechVehicle.BP_FigmentTestMechVehicle_C" (Math|Conversions|ToTransform(Vector) (+ (Transformation|GetActorLocation (Pawn|GetControlledPawn (CheatManager|GetPlayerController))) (* (Transformation|GetActorForwardVector (Pawn|GetControlledPawn (CheatManager|GetPlayerController))) "10.000000,10.000000,10.000000"))) "Undefined" "MultiplyWithRoot" 0 0)
  (Development|PrintText "NSLOCTEXT(\"[7414E29BD63F8D88A7072A7F57793B16]\", \"7EC9A53B44215A9C2B3598B63358EC37\", \"Spawning Mech\")" true true "(R=1.000000,G=0.000000,B=0.151955,A=1.000000)"))

(fn SpawnMechDrop
  (Game|SpawnActorfromClass "/Figment_OstrichVehicle/Blueprints/BGA_Figment_Ostrich_Drop.BGA_Figment_Ostrich_Drop_C" (Math|Conversions|ToTransform(Vector) (+ (Transformation|GetActorLocation (Pawn|GetControlledPawn (CheatManager|GetPlayerController))) (* (Transformation|GetActorForwardVector (Pawn|GetControlledPawn (CheatManager|GetPlayerController))) "700.000000,700.000000,30000.000000"))) "Undefined" "MultiplyWithRoot" 0 0 0)
  (Development|PrintText "NSLOCTEXT(\"[7414E29BD63F8D88A7072A7F57793B16]\", \"7EC9A53B44215A9C2B3598B63358EC37\", \"Spawning Mech\")" true true "(R=1.000000,G=0.000000,B=0.151955,A=1.000000)"))
